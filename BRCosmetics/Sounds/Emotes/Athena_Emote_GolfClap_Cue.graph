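# Sound cue for the Athena golf clap emote, playing one of seven clap recordings at random.
# A Random node at the root picks among seven wave players with equal weight, feeding straight to output at reduced volume.
# It uses the shared emote attenuation, music emote sound class and emote foley concurrency set.
# It also has a send to the music audio analysis submix.
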SoundCue /BRCosmetics/Sounds/Emotes/Athena_Emote_GolfClap_Cue
root=Random_0
nodes [
 {
  id=Random_0
  children [WavePlayer_0,WavePlayer_1,WavePlayer_2,WavePlayer_3,WavePlayer_4,WavePlayer_5,WavePlayer_6]
  properties {weights [1,1,1,1,1,1,1]}
 }
 {id=WavePlayer_0, properties {soundWaveAssetPtr=&/Athena_Emote_GolfClap_01}}
 {id=WavePlayer_1, properties {soundWaveAssetPtr=&/Athena_Emote_GolfClap_02}}
 {id=WavePlayer_2, properties {soundWaveAssetPtr=&/Athena_Emote_GolfClap_03}}
 {id=WavePlayer_3, properties {soundWaveAssetPtr=&/Athena_Emote_GolfClap_04}}
 {id=WavePlayer_4, properties {soundWaveAssetPtr=&/Athena_Emote_GolfClap_05}}
 {id=WavePlayer_5, properties {soundWaveAssetPtr=&/Athena_Emote_GolfClap_06}}
 {id=WavePlayer_6, properties {soundWaveAssetPtr=&/Athena_Emote_GolfClap_07}}
]
properties {
 attenuationSettings=&Fort_Emotes_Athena_Att
 maxDistance=1600
 soundClassObject=&Fort_Music_Emote
 volumeMultiplier=0.5
 concurrencyOverrides {resolutionRule=PreventNew}
 concurrencySet [/Game/Sounds/Concurrency_Rules/Athena_EmoteFoley_Conc]
 soundSubmixSends [{minSendLevel=0, soundSubmix=&Fortnite_AudioAnalysis_Music}]
}
