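# Sound cue that plays music for the SpikeIt emote, using first-person emote music class and property sheet settings.
# A mixer runs two delayed branches: the intro starts after a brief delay, and a looping track enters much later on a fixed offset.
# Sequencing relies on delay timing alone.
# An emote music concurrency set applies, and output also feeds a music audio-analysis submix.
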SoundCue /BRCosmetics/Sounds/Emotes/Athena_Emote_SpikeIt_Music_Cue
root=Mixer_1
nodes [
 {id=Delay_0, children [WavePlayer_3], properties {delayMax=0.15, delayMin=0.15}}
 {id=Delay_2, children [WavePlayer_2], properties {delayMax=34.64, delayMin=34.64}}
 {id=Mixer_1, children [Delay_0,Delay_2], properties {inputVolume [1,1.01]}}
 {id=WavePlayer_2, properties {bLooping=true, soundWaveAssetPtr=&/Touchdown_SpikeIt_Emote_Music_01_Loop}}
 {id=WavePlayer_3, properties {soundWaveAssetPtr=&/Touchdown_SpikeIt_Emote_Music_Intro_01}}
]
properties {
 audioPropertiesSheet=/BRCosmetics/Sounds/Settings/PropertySheets/Emotes/APS_Cosmetic_Emote_Music_1P
 maxDistance=2097152
 soundClassObject=/Game/Sounds/FortSoundClassesAndModes/Fort_Music_Emote_1P_PSM
 volumeMultiplier=1
 concurrencyOverrides {resolutionRule=PreventNew}
 concurrencySet [/Game/Sounds/Concurrency_Rules/Athena_EmoteMusic_1P_Conc]
 soundSubmixSends [{minSendLevel=0, soundSubmix=&Fortnite_AudioAnalysis_Music}]
}
