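# Foley sound cue for the ThumbDecide emote in Battle Royale (Athena), playing a stereo player swing whoosh.
# An Enveloper wraps the single wave player, sweeping pitch downward over the first half second, with lowered overall volume and pitch.
# It uses emote foley class, property sheet, attenuation and concurrency assets, and carries a send to the music audio-analysis submix.
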SoundCue /BRCosmetics/Sounds/Emotes/Athena_Emote_ThumbDecide_Foley_Cue
root=Enveloper_0
nodes [
 {
  id=Enveloper_0
  children [WavePlayer_6]
  properties {pitchCurve {editorCurveData {keys [<at 0 1.5>,<at 0.5 0.7>]}}, volumeCurve {editorCurveData {keys [<at 0 1>]}}}
 }
 {id=WavePlayer_6, properties {soundWaveAssetPtr=/Game/Sounds/Fort_Foley_Sounds/Whooshes/player_swing_whoosh_stereo_02}}
]
properties {
 attenuationSettings=&Fort_Emotes_Athena_Att
 audioPropertiesSheet=/BRCosmetics/Sounds/Settings/PropertySheets/Emotes/APS_Cosmetic_Emote_Foley
 maxDistance=1600
 pitchMultiplier=0.8
 soundClassObject=/Game/Sounds/FortSoundClassesAndModes/Fort_Foley_Emote
 volumeMultiplier=0.3
 concurrencyOverrides {resolutionRule=PreventNew}
 concurrencySet [/Game/Sounds/Concurrency_Rules/Athena_EmoteFoley_Conc]
 soundSubmixSends [{minSendLevel=0, soundSubmix=&Fortnite_AudioAnalysis_Music}]
}
