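# Sound cue that supplies cloth-movement style foley for the Thumb Raise emote, built from two reused player run foley waves.
# A mixer layers a quieter immediate wave with a full-level second wave after a short fixed delay, with raised pitch and reduced overall volume.
# It uses emote attenuation, foley sound class and concurrency, and has a send to the music audio-analysis submix.
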SoundCue /BRCosmetics/Sounds/Emotes/Athena_Emote_ThumbRaise_Foley_Cue
root=Mixer_0
nodes [
 {id=Delay_0, children [WavePlayer_1], properties {delayMax=0.1, delayMin=0.1}}
 {id=Mixer_0, children [WavePlayer_0,Delay_0], properties {inputVolume [0.4,1]}}
 {id=WavePlayer_0, properties {soundWaveAssetPtr=/Game/Sounds/Fort_Foley_Sounds/Moves/player_run_foley_10}}
 {id=WavePlayer_1, properties {soundWaveAssetPtr=/Game/Sounds/Fort_Foley_Sounds/Moves/player_run_foley_09}}
]
properties {
 attenuationSettings=&Fort_Emotes_Athena_Att
 audioPropertiesSheet=/BRCosmetics/Sounds/Settings/PropertySheets/Emotes/APS_Cosmetic_Emote_Foley
 maxDistance=1600
 pitchMultiplier=1.2
 soundClassObject=/Game/Sounds/FortSoundClassesAndModes/Fort_Foley_Emote
 volumeMultiplier=0.6
 concurrencyOverrides {resolutionRule=PreventNew}
 concurrencySet [/Game/Sounds/Concurrency_Rules/Athena_EmoteFoley_Conc]
 soundSubmixSends [{minSendLevel=0, soundSubmix=&Fortnite_AudioAnalysis_Music}]
}
