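# Sound cue for the third-person stomp in the Take The Elf emote, layering leg-swing foley, a footstep stomp and a bell stomp.
# A mixer combines three random nodes with six equally weighted variations each; the swing plays immediately, while both stomp layers follow after short fixed, slightly staggered delays.
# It uses emote attenuation, music-emote sound class and emote foley concurrency, and sends to a music audio-analysis submix.
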SoundCue /BRCosmetics/Sounds/Emotes/Emote_TakeTheElf_Stomp3p_Cue
root=Mixer_0
nodes [
 {id=Delay_0, children [Random_0], properties {delayMax=0.25, delayMin=0.25}}
 {id=Delay_1, children [Random_2], properties {delayMax=0.27, delayMin=0.27}}
 {id=Mixer_0, children [Delay_0,Random_1,Delay_1], properties {inputVolume [1,1,1]}}
 {id=Random_0, children [WavePlayer_0,WavePlayer_2,WavePlayer_3,WavePlayer_4,WavePlayer_5,WavePlayer_6], properties {weights [1,1,1,1,1,1]}}
 {
  id=Random_1
  children [WavePlayer_12, WavePlayer_7, WavePlayer_8, WavePlayer_9, WavePlayer_10, WavePlayer_11]
  properties {weights [1,1,1,1,1,1]}
 }
 {
  id=Random_2
  children [WavePlayer_13, WavePlayer_14, WavePlayer_15, WavePlayer_16, WavePlayer_17, WavePlayer_18]
  properties {weights [1,1,1,1,1,1]}
 }
 {id=WavePlayer_0, properties {soundWaveAssetPtr=&/Emote_TakeTheElf_SFX_FS_Stomp_06}}
 {id=WavePlayer_10, properties {soundWaveAssetPtr=&/Emote_TakeTheElf_SFX_Foly_LegSwing_05}}
 {id=WavePlayer_11, properties {soundWaveAssetPtr=&/Emote_TakeTheElf_SFX_Foly_LegSwing_06}}
 {id=WavePlayer_12, properties {soundWaveAssetPtr=&/Emote_TakeTheElf_SFX_Foly_LegSwing_01}}
 {id=WavePlayer_13, properties {soundWaveAssetPtr=&/Emote_TakeTheElf_SFX_Foly_BellStomp_06}}
 {id=WavePlayer_14, properties {soundWaveAssetPtr=&/Emote_TakeTheElf_SFX_Foly_BellStomp_01}}
 {id=WavePlayer_15, properties {soundWaveAssetPtr=&/Emote_TakeTheElf_SFX_Foly_BellStomp_02}}
 {id=WavePlayer_16, properties {soundWaveAssetPtr=&/Emote_TakeTheElf_SFX_Foly_BellStomp_03}}
 {id=WavePlayer_17, properties {soundWaveAssetPtr=&/Emote_TakeTheElf_SFX_Foly_BellStomp_04}}
 {id=WavePlayer_18, properties {soundWaveAssetPtr=&/Emote_TakeTheElf_SFX_Foly_BellStomp_05}}
 {id=WavePlayer_2, properties {soundWaveAssetPtr=&/Emote_TakeTheElf_SFX_FS_Stomp_01}}
 {id=WavePlayer_3, properties {soundWaveAssetPtr=&/Emote_TakeTheElf_SFX_FS_Stomp_02}}
 {id=WavePlayer_4, properties {soundWaveAssetPtr=&/Emote_TakeTheElf_SFX_FS_Stomp_03}}
 {id=WavePlayer_5, properties {soundWaveAssetPtr=&/Emote_TakeTheElf_SFX_FS_Stomp_04}}
 {id=WavePlayer_6, properties {soundWaveAssetPtr=&/Emote_TakeTheElf_SFX_FS_Stomp_05}}
 {id=WavePlayer_7, properties {soundWaveAssetPtr=&/Emote_TakeTheElf_SFX_Foly_LegSwing_02}}
 {id=WavePlayer_8, properties {soundWaveAssetPtr=&/Emote_TakeTheElf_SFX_Foly_LegSwing_03}}
 {id=WavePlayer_9, properties {soundWaveAssetPtr=&/Emote_TakeTheElf_SFX_Foly_LegSwing_04}}
]
properties {
 attenuationSettings=&Fort_Emotes_Athena_Att
 maxDistance=1600
 soundClassObject=&Fort_Music_Emote
 volumeMultiplier=1
 concurrencyOverrides {maxCount=3, resolutionRule=StopOldest, volumeScale=0.75}
 concurrencySet [/Game/Sounds/Concurrency_Rules/Athena_EmoteFoley_Conc]
 soundSubmixSends [{minSendLevel=0, soundSubmix=&Fortnite_AudioAnalysis_Music}]
}
