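# Sound cue for the texting emote's phone notification sound.
# A random node chooses among three notification waves and an empty fourth input, all equally weighted, so some plays can be silent.
# It uses the Athena emote attenuation and emote music sound class, with boosted volume.
# An emote foley concurrency set is assigned, and a submix send routes to the music audio analysis submix.
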SoundCue /BRCosmetics/Sounds/Emotes/Emote_Texting_Notification_Cue
root=Random_0
nodes [
 {id=Random_0, children [WavePlayer_1,WavePlayer_2,WavePlayer_3,null], properties {weights [1,1,1,1]}}
 {id=WavePlayer_1, properties {soundWaveAssetPtr=&/Emote_Texting_Notification_01}}
 {id=WavePlayer_2, properties {soundWaveAssetPtr=&/Emote_Texting_Notification_02}}
 {id=WavePlayer_3, properties {soundWaveAssetPtr=&/Emote_Texting_Notification_03}}
]
properties {
 attenuationSettings=&Fort_Emotes_Athena_Att
 maxDistance=1600
 soundClassObject=&Fort_Music_Emote
 volumeMultiplier=2
 concurrencyOverrides {resolutionRule=StopOldest}
 concurrencySet [/Game/Sounds/Concurrency_Rules/Athena_EmoteFoley_Conc]
 soundSubmixSends [{minSendLevel=0, soundSubmix=&Fortnite_AudioAnalysis_Music}]
}
