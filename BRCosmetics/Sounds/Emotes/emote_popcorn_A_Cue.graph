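# Sound cue for the popcorn emote, playing one of five popcorn toss sounds.
# An equally weighted Random node picks the wave and feeds a Modulator with fixed pitch, then a Mixer whose second input is empty.
# It uses emote attenuation, the emote music sound class and emote foley concurrency, with a submix send to music audio analysis.
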SoundCue /BRCosmetics/Sounds/Emotes/emote_popcorn_A_Cue
root=Mixer_0
nodes [
 {id=Mixer_0, children [Modulator_1,null], properties {inputVolume [0.5,0.5]}}
 {id=Modulator_1, children [Random_0], properties {pitchMax=1, pitchMin=1}}
 {id=Random_0, children [WavePlayer_1,WavePlayer_2,WavePlayer_3,WavePlayer_4,WavePlayer_5], properties {weights [1,1,1,1,1]}}
 {id=WavePlayer_1, properties {soundWaveAssetPtr=&/Emote_PopcornToss_04}}
 {id=WavePlayer_2, properties {soundWaveAssetPtr=&/Emote_PopcornToss_01}}
 {id=WavePlayer_3, properties {soundWaveAssetPtr=&/Emote_PopcornToss_02}}
 {id=WavePlayer_4, properties {soundWaveAssetPtr=&/Emote_PopcornToss_03}}
 {id=WavePlayer_5, properties {soundWaveAssetPtr=&/Emote_PopcornToss_05}}
]
properties {
 attenuationSettings=&Fort_Emotes_Athena_Att
 maxDistance=1600
 soundClassObject=&Fort_Music_Emote
 volumeMultiplier=0.9
 concurrencySet [/Game/Sounds/Concurrency_Rules/Athena_EmoteFoley_Conc]
 soundSubmixSends [{minSendLevel=0, soundSubmix=&Fortnite_AudioAnalysis_Music}]
}
